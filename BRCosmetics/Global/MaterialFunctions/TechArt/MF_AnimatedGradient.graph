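# Material function that outputs an animated, scrolling color gradient, scaled by a Mask input.
# An input gradient map or selectable UV axis, optionally inverted, is offset by time and random noise, then tiled and wrapped with Frac.
# The result samples a curve atlas row or gradient texture, with multiplied color and noise added.
# The optional UV crop remap wires each Min value to both target bounds, leaving Max unused.
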MaterialFunction /BRCosmetics/Global/MaterialFunctions/TechArt/MF_AnimatedGradient
graph {nodes [
  n1 = Multiply(A=Time(), B=ScalarParameter() {defaultValue=0.5, group=AnimatedGradient, parameterName=Speed}) {constB=0.1}
  n2 = StaticSwitchParameter(True=NamedRerouteUsage(), False=FunctionInput() {inputName=Gradient Map, inputType=FunctionInput_Scalar}) {group=AnimatedGradient, parameterName=Use_UV_As_GradientMap}
  n3 = MaterialFunctionCall!(NoisePower=ScalarParameter() {defaultValue=25, group=AnimatedGradient, parameterName=Noise Power}, Coords=TextureCoordinate()) {materialFunction=/Game/Global/MaterialFunctions/TechArt/MF_RandomNoise}
  n4 = Add(
   A=LinearInterpolate(A=@n2, B=OneMinus(None=@n2), Alpha=ScalarParameter() {group=AnimatedGradient, parameterName=Invert_GradientMap, sliderMax=1})
   B=@n3.Result
  ) {constB=0}
  n5 = Frac(None=Multiply(A=Add(A=@n1, B=@n4), B=ScalarParameter() {defaultValue=1, group=AnimatedGradient, parameterName=Animated_Gradient_Tiling}))
  n6 = CurveAtlasRowParameter(CurveTime=@n5) {
   atlas=&AthenaCurveAtlas01
   curve=/Game/Athena/LinearColorCurves/Mello_Rainbow
   defaultValue=52
   group=AnimatedGradient
   parameterName=Curve
  }
  n7 = AppendVector(A=@n5, B=Constant())
  n8 = ScalarParameter() {defaultValue=1, group=AnimatedGradient, parameterName=Color_Multiplier}
  n9 = Multiply(
   A=StaticSwitchParameter(
    True=@n6
    False=TextureSampleParameter2D(UVs=@n7) {group=AnimatedGradient, parameterName=GradientTex, texture=/Game/Creative/BuildingActors/Props/Textures/CP_White_Texutre}.RGB
   ) {group=AnimatedGradient, parameterName=Use_Curve}
   B=@n8
  )
  n10 = Multiply(A=@n3.Result, B=@n8)
  FunctionOutput(None=LinearInterpolate(B=Add(A=@n9, B=@n10) {constB=0}, Alpha=FunctionInput() {inputName=Mask, inputType=FunctionInput_Scalar}) {constB=0})
  n11 = TextureCoordinate()
  n12 = TextureCoordinate() {coordinateIndex=1}
  n13 = StaticSwitchParameter(
   True=NamedRerouteUsage()
   False=NamedRerouteDeclaration(None=StaticSwitchParameter(True=@n11, False=@n12) {defaultValue=true, desc=Use UV0 otherwise use UV1, group=AnimatedGradient, parameterName=Use_UV0}) {name=UVSelector, nodeColor=<rgba 0 1 0.7656 1>}
  ) {group=AnimatedGradient, parameterName=Use_CropUV}
  NamedRerouteDeclaration(None=StaticSwitchParameter(True=ComponentMask(None=@n13) {r=true}, False=ComponentMask(None=@n13) {g=true}) {group=AnimatedGradient, parameterName=Use_X_For_Gradient_Direction}) {name=UV_GradientMap, nodeColor=<rgba 0.8516 0 1 1>}
  n14 = NamedRerouteUsage()
  n15 = ComponentMask(None=@n14) {r=true}
  n16 = Constant()
  n17 = Constant() {r=1}
  n18 = VectorParameter() {group=AnimatedGradient, parameterName=Crop_UV, channelNames {a=Y_Max, b=Y_Min, g=X_Max, r=X_Min}, defaultValue=<rgba 0 1 0 1>}
  n19 = ComponentMask(None=@n14) {g=true}
  NamedRerouteDeclaration(None=AppendVector(
    A=MaterialFunctionCall(Input=@n15, Input Low=@n16, Input High=@n17, Target Low=@n18.X_Min, Target High=@n18.X_Min) {materialFunction=&RemapValueRange}.Result
    B=MaterialFunctionCall(Input=@n19, Input Low=@n16, Input High=@n17, Target Low=@n18.Y_Min, Target High=@n18.Y_Min) {materialFunction=&RemapValueRange}.Result
   )) {name=CoppedUVs, nodeColor=<rgba 0.0078 1 0 1>}
 ]}
