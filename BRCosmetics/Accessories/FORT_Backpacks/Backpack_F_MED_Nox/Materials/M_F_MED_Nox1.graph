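# Surface material for the Nox backpack cosmetic, built on the shared character PreFX and PostFX functions plus a cinematic post-FX function.
# Between them, static switches optionally add metal LUT, cloth fuzz and silk effects driven by a SkinFX mask texture, followed by a cloth-sim world position offset function.
# An animated gradient, driven by a gradient texture, is blended into base color and emissive through a selectable mask channel.
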Material /BRCosmetics/Accessories/FORT_Backpacks/Backpack_F_MED_Nox/Materials/M_F_MED_Nox1
graph {
 properties {
  bUseMaterialAttributes=true
  bUsedWithMeshDeformer=true
  bUsedWithSkeletalMesh=true
  pixelDepthOffsetMode=PDOM_Legacy
  usedShadingModels=Default Lit
 }
 nodes [
  n1 = MaterialFunctionCall() {materialFunction=&BaseFN_CharacterFunction_PreFX}
  n2 = MaterialFunctionCall(AO=@n1.AO, Distance Blend=@n1.AO, Fresnel=@n1.AO, Gmap=@n1.AO, Pre Skinned Local Position=@n1.AO, Material=NamedRerouteUsage()) {materialFunction=&BaseFN_CharacterFunction_PostFX}
  MaterialFunctionCall(In=@n2.Result) {materialFunction=&CINE_MF_CharPostFX}
  n3 = NamedRerouteUsage()
  n4 = StaticSwitchParameter(
   True=MaterialFunctionCall(Mask=@n3, Materials=@n1.Result) {materialFunction=&MF_MetalsLUT}.Result
   False=@n1.Result
  ) {group=Skin FX, parameterName=UseMetalLUT}
  n5 = StaticSwitchParameter(
   True=MaterialFunctionCall!(Alpha=ChannelMaskParameter(None=@n3) {group=Skin FX, parameterName=Cloth Channel}.R, Gmap=@n1.Gmap, Material=@n4) {materialFunction=/Game/Weapons/WeaponSkins/MaterialFunction/ClothFuzz}.Result
   False=@n4
  ) {group=Skin FX, parameterName=Use Cloth fuzz}
  n6 = MaterialFunctionCall!(Alpha=ChannelMaskParameter(None=@n3) {group=Skin FX, parameterName=Silk Channel}.R, Material=@n5) {materialFunction=/Game/Weapons/WeaponSkins/MaterialFunction/MF_Silk}
  NamedRerouteDeclaration(None=MaterialFunctionCall!(Material=StaticSwitchParameter(True=@n6.Result, False=@n5) {group=Skin FX, parameterName=Use Silk}, SkinFX_Mask=@n3) {materialFunction=/Game/Global/MaterialFunctions/MF_ClothSimWPO_v2}.Result) {name=Material, nodeColor=<rgba 0 0.1562 1 1>}
  NamedRerouteDeclaration(None=TextureSampleParameter2D() {group=Skin FX, parameterName=SkinFX_Mask}.RGBA) {name=SkinFX_Mask, nodeColor=<rgba 0.9842 0 1 1>}
  n7 = GetMaterialAttributes(None=NamedRerouteUsage()) {attributeGetTypes [#1, #2]}
  n8 = ChannelMaskParameter(None=NamedRerouteUsage()) {group=Skin FX, parameterName=AnimatedGradient_Mask_Channel}
  n9 = MaterialFunctionCall!(
   Mask=@n8.R
   Gradient Map=ChannelMaskParameter(None=TextureSampleParameter2D() {group=AnimatedGradient, parameterName=Gradient_Tex}.RGBA) {group=AnimatedGradient, parameterName=Gradient_Tex_Channel}.R
  ) {materialFunction=/BRCosmetics/Global/MaterialFunctions/TechArt/MF_AnimatedGradient}
  NamedRerouteDeclaration(None=SetMaterialAttributes!(
    MaterialAttributes=@n7.MaterialAttributes
    Base Color=LinearInterpolate(A=@n7.Base Color, B=@n9.Result, Alpha=@n8.R)
    Emissive Color=LinearInterpolate(A=@n7.Emissive Color, B=@n9.Result, Alpha=@n8.R)
   ) {attributeSetTypes [#1, #2]}) {name=Out, nodeColor=<rgba 1 0 0.0469 1>}
 ]
}
