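# Physics asset for the F_MED_Nox backpack cosmetic, giving physics bodies to the pack and its two dynamic ropes.
# One capsule covers the pack, and each left and right rope is a chain of two small capsules, slightly thicker on the right.
# Constraints link each chain to the pack with a tighter swing cone at the root, looser at the second segment, and almost no twist.
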PhysicsAsset /BRCosmetics/Accessories/FORT_Backpacks/Backpack_F_MED_Nox/Meshes/F_MED_Nox_Pack_Physics
bodies {
 pack [{shapeType=Capsule, length=9, radius=15, shapeName=pack_capsule, center=<0 14.1749 -17.2958>, rotation=<rot -90 90 -90.0007>}]
 rope_dyn_l [{shapeType=Capsule, length=3, radius=1, shapeName=rope_dyn_l_capsule, center=<-2.0017 0 0>, rotation=<rot 90 0 0>}]
 rope_dyn_l_1 [{shapeType=Capsule, length=4, radius=1, shapeName=rope_dyn_l_capsule, center=<-2.6421 0 0>, rotation=<rot 90 0 0>}]
 rope_dyn_r [{shapeType=Capsule, length=3, radius=2, shapeName=rope_dyn_l_capsule, center=<2.0399 0 0>, rotation=<rot 90 0 0>}]
 rope_dyn_r_1 [{shapeType=Capsule, length=3.5921, radius=2, shapeName=rope_dyn_l_capsule, center=<3.9498 0 0>, rotation=<rot 90 0 0>}]
}
constraints [
 {bone1Name=rope_dyn_l_1, bone2Name=rope_dyn_l, swing1LimitDegrees=40, swing2LimitDegrees=40, twistLimitDegrees=0.1}
 {bone1Name=rope_dyn_l, bone2Name=pack, swing1LimitDegrees=18, swing2LimitDegrees=18, twistLimitDegrees=0.1}
 {bone1Name=rope_dyn_r, bone2Name=pack, swing1LimitDegrees=18, swing2LimitDegrees=18, twistLimitDegrees=0.1}
 {bone1Name=rope_dyn_r_1, bone2Name=rope_dyn_r, swing1LimitDegrees=40, swing2LimitDegrees=40, twistLimitDegrees=0.1}
]
