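# Animation blueprint for the Nox backpack cosmetic, adding physics sway and posture correction.
# A RigidBody node feeds two pack bone transforms: a fixed offset and scale, then roll and translation mapped from the owner's spine bend.
# Weapon, customization and skydiving events drive the "Hide Element 01" material parameter, setting it to 1 only while a Nox-tagged pickaxe is equipped outside skydiving.
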AnimBlueprint /BRCosmetics/Accessories/FORT_Backpacks/Backpack_F_MED_Nox/Meshes/F_MED_Nox_Pack_AnimBP
target_skeleton=&/F_MED_Nox_Pack_Skeleton
graphs [
 AnimGraph {nodes [
   n1 = RigidBody(ComponentPose=Local To Component(LocalPose=Input Pose())) {
    bClampLinearTranslationLimitToRefPose=true
    bOverrideWorldGravity=true
    overridePhysicsAsset=&/F_MED_Nox_Pack_Physics
    alphaBoolBlend {blendOption=Linear}
    baseBoneRef {boneName=root}
    componentAppliedLinearAccClamp=<3000 3000 3000>
    componentLinearAccScale=<0.4 0.4 0.4>
    componentLinearVelScale=<0.3 0.3 0.3>
    overrideWorldGravity {z=-980}
   }
   n2 = "Transform (Modify) Bone - Bone: pack"(ComponentPose=@n1, Translation="0,-0.5,-3.5", Rotation="0,0,10", Scale="1.3,1.3,1.3") {
    rotationMode=BMM_Additive
    scaleMode=BMM_Replace
    translationMode=BMM_Additive
    alphaBoolBlend {blendOption=Linear}
    boneToModify {boneName=pack}
    rotation {roll=10}
    scale=<1.3 1.3 1.3>
    translation {y=-0.5, z=-3.5}
   }
   n3 = "Transform (Modify) Bone - Bone: pack"(ComponentPose=@n2, Translation=Get backpack_translation(), Rotation=Get backpack_rotate()) {
    rotationMode=BMM_Additive
    scaleMode=BMM_Additive
    translationMode=BMM_Additive
    alphaBoolBlend {blendOption=Linear}
    boneToModify {boneName=pack}
    rotation {roll=10}
    translation {y=-0.5, z=-2.5}
   }
   Output Pose(Result=Component To Local(ComponentPose=@n3)) {layerGroup=None}
  ]}
 EventGraph {nodes [
   Event BlueprintUpdateAnimation()
   n1 = GetOwnerSkeletalMeshComponent()
   n2 = BreakRotator(InRot=BreakTransform(InTransform=GetDeltaTransformFromRefPose(self=@n1, BoneName=spine_05, BaseName=spine_01)).Rotation)
   > Set backpack_rotate(backpack_rotate=MakeRotator(Roll=MapRangeClamped(Value=@n2.Yaw, InRangeA=0, InRangeB=-30, OutRangeB=25)))
   > Set backpack_translation(backpack_translation=MakeVector(Y=MapRangeClamped(Value=@n2.Yaw, InRangeA=0, InRangeB=-30, OutRangeB=-5)))
   n3 = Event OnWeaponChanged()
   n4 = Get Dynamic Material Instance(execute=[@n3.then, @n12.then], SkeletalMeshComponent=GetPartSkeletalMeshComponent())
   > n5 = Sequence()
   n6 = PickaxeItemDefinitionHasAnyMetaTag(Tags=Make Array("[0]"=(TagName="Cosmetics.MetaTags.Nox")), bRequiresEquipped=true)
   > n7 = Branch(Condition=@n6)
   > n8 = Branch(Condition=Get bIsSkyDiving())
   SetScalarParameterValue(execute=[@n10.then, @n8.then, @n9.else, @n7.else], self=@n4, ParameterName=Hide Element 01, Value=0)
   SetScalarParameterValue(execute=[@n8.else, @n9.then], self=@n4, ParameterName=Hide Element 01, Value=1)
   n9 = Branch(execute=[@n11.then, @n5.then_1], Condition=@n6)
   n10 = Event OnBeginSkydiving()
   n11 = Event OnEndSkydiving()
   n12 = Event OnCharacterCustomizationLoaded()
   n13 = Sequence()
   > n14 = Cast To PlayerPawn_Athena(Object=TryGetPawnOwner())
   > Set As Player Pawn Athena(As Player Pawn Athena:Player Pawn Athena Object Reference=@n14) -> (Output_Get:Player Pawn Athena Object Reference)
   n15 = Get Dynamic Material Instance(execute=@n13.then_1, SkeletalMeshComponent=GetPartSkeletalMeshComponent())
   > Set BackPakMaterial(BackPakMaterial:Material Instance Dynamic Object Reference=@n15) -> (Output_Get:Material Instance Dynamic Object Reference)
  ]}
 "on change " {nodes [
   n1 = Inputs!() -> (exec:Exec, on change:bool=~)
   n2 = Local Boolean!() -> (Variable:bool=~) {is=title K2Node_TemporaryVariable}
   > Branch(Condition="Not Equal (Boolean)"(A=@n1, B=@n2) -> (ReturnValue=~))
   > Assign!(execute:Exec, Variable:bool=@n2, Value:bool=@n1) -> (then:Exec) {is=title K2Node_AssignmentStatement}
   > Outputs!(out:Exec)
  ]}
]
variables [
 As Player Pawn Athena {type=PlayerPawn_Athena_C}
 backpack_rotate {type=Rotator}
 backpack_translation {type=Vector}
 BackPakMaterial {type=MaterialInstanceDynamic}
 RigidBodyAlpha {type=number}
]
